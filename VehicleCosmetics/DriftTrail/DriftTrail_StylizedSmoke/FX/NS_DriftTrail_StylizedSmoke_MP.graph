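# Niagara drift trail cosmetic that emits stylized colored smoke and wheel wrap effects from a vehicle's wheels.
# Cone and Trail sprite emitters spawn at wheel sockets with speed-driven velocity, adding curl noise and camera fade respectively.
# WheelWrap burst-spawns socket-following meshes scaled to the wheels, with waveform-animated purple/orange outlines when unpainted.
# All three emitters take painted-variant color overrides and use per-wheel partitions to disable tire FX individually.
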NiagaraSystem /VehicleCosmetics/DriftTrail/DriftTrail_StylizedSmoke/FX/NS_DriftTrail_StylizedSmoke_MP
dependencies {
 usedDataInterfaces [
  &NiagaraDataInterfaceCurve
  &NiagaraDataInterfaceArrayColor
  &NiagaraDataInterfaceCamera
  &NiagaraDataInterfaceSkeletalMesh
 ]
 usedDynamicInputs []
 usedModules [
  &self:SystemSpawnScript.SystemScriptSource.SystemScriptGraph.NiagaraNodeAssignment_0.SetVariables_#1
  &FN_UnpackVehicleFlags
  /VehicleCosmetics/FX/Niagara/Modules/VC_SetVehicleSpeedAttribs
  &SystemState
  /Game/Effects/Niagara/Modules/Vehicle/FN_VehicleFlags_Cheatcodes
 ]
}
emitters {
 NE_SS_Cone {
  emitterSpawnScript [/SetVariables(Emitter.Linear Color:LinearColor=&MakeCustomLinearColorFromBool(
     Bool=&SetBoolByFloatComparison(A=&MakeFloatFromInt(Int=$User.ColorVariantID), B=14, Comparison Type=::NewEnumerator4)
     True Color=<rgba 1 1 1 1>
     False Color=<rgba 0 0 0 1>
    )) {isSetParameters=true}]
  emitterUpdateScript [
   &EmitterState()
   &SpawnRate(SpawnRate=60, Use Spawn Probability=false)
  ]
  particleSpawnScript [
   &VC_PaintedVariantsOverride(
    Black=<rgba 0.0296 0.0343 0.0382 1>
    BurntSienna=<rgba 0.25 0.042 0.0377 1>
    Cobalt=<rgba 0.0437 0.1384 1 1>
    Crimson=<rgba 0.7231 0.0242 0.0437 1>
    DefaultColor=&SelectLinearColorFromArray(
     Array Sampling Mode=::NewEnumerator0
     Direct Array Index=?
     Color Selection Array=DataInterface() {propertyValues {ColorData [<rgba 0.8 0.424 0.048 1>, <rgba 0.4951 0.0063 0.83 1>]}}
     Random Seed=0
    )
    ForestGreen=<rgba 0.003 1 0.0931 1>
    Grey=<rgba 0.3419 0.3467 0.3564 1>
    Lime=<rgba 0.5583 0.9823 0.1384 1>
    Orange=<rgba 0.6795 0.2051 0.0307 1>
    Pink=<rgba 0.9047 0.0497 0.7157 1>
    Purple=<rgba 0.261 0.0025 0.65 1>
    R1 Black:LinearColor=<rgba 0 0 0 1>
    R1 BurntSienna:LinearColor=<rgba 0 0 0 1>
    R1 Cobalt:LinearColor=<rgba 0 0 0 1>
    R1 Crimson:LinearColor=<rgba 0 0 0 1>
    R1 DefaultColor=<rgba 0 0 0 1>
    R1 ForestGreen:LinearColor=<rgba 0 0 0 1>
    R1 Gold:LinearColor=<rgba 0 0 0 1>
    R1 Grey:LinearColor=<rgba 0 0 0 1>
    R1 Lime:LinearColor=<rgba 0 0 0 1>
    R1 Orange:LinearColor=<rgba 0 0 0 1>
    R1 Pink:LinearColor=<rgba 0 0 0 1>
    R1 Purple:LinearColor=<rgba 0 0 0 1>
    R1 Saffron:LinearColor=<rgba 0 0 0 1>
    R1 SkyBlue:LinearColor=<rgba 0 0 0 1>
    Saffron=<rgba 0.8595 0.9 0.1136 1>
    SkyBlue=<rgba 0.0958 0.8513 0.9 1>
    TitaniumWhite=<rgba 0 0 0 1>
    Use R1 Painted Overrides=true
   )
   &InitializeParticle(
    Color Mode=::NewEnumerator0
    Lifetime=0.9
    Sprite Rotation Mode=::NewEnumerator2
    Sprite Size Mode=::NewEnumerator1
    Uniform Sprite Size Max=55
    Uniform Sprite Size Min=25
    UsePositionOffset=false
    Write Material Random=true
   )
   &PartitionParticles()
   &SkeletalMeshLocation(
    Bone / Socket Index=&Add_Integer(A=$Particles.PartitionParticles.Partition, B=4)
    Bone / Socket Sampling Mode=::NewEnumerator5
    Interpolate Mesh Sampling=false
    MaskSpawnGroup=false
    Mesh Sampling Type=::NewEnumerator6
    Sampled Position Offset=&AddVectorToPosition(Position=$User.offset, Vector=<-45 0 -20>)
    Skeletal Mesh=$System.SkeletalMesh
   )
   &ShapeLocation(Hemisphere Distribution=<5 1>, MaskSpawnGroup=false, Override Seed=false, Sphere Radius=12, UseOwnerScale=true)
   &AddVelocity(Velocity=&VC_Mask_Vectors(
     FrontEnd Value:Vector3f=&RandomRangeVector(Minimum=<-70 -30 5>, Maximum=<-80 30 20>)
     blsFrontEndPreview:NiagaraBool=true
     SpeedMin Value:Vector3f=<-800 0 0>
     SpeedMax Value:Vector3f=<-1 0 0>
     Speed Alpha:NiagaraFloat=$System.Speed_Alpha
    ))
  ]
  particleUpdateScript [
   &FN_DisableTireFX(
    Mask Index Attribute:NiagaraInt32=$Particles.PartitionParticles.Partition
    Mask Index FL=2
    Mask Index FR=3
    Mask Index RL=0
    Mask Index RR=1
   )
   &ParticleState()
   &ScaleSpriteSize(Uniform Curve Sprite Scale=DataInterface() {propertyValues {Curve {
       defaultValue=0
       keys [
        {arriveTangent=0.1044, interpMode=RCIM_Cubic, leaveTangent=0.1044, tangentMode=RCTM_User, time=0.5, value=1}
        {interpMode=RCIM_Cubic, tangentMode=RCTM_User, time=1, value=0}
        {interpMode=RCIM_Cubic, tangentMode=RCTM_User, time=5, value=1}
       ]
      }}})
   &ScaleColor(Scale Alpha=5, ScaleA=true, ScaleRGB=true)
   &SpriteRotationRate(Rotation Rate=&Multiply_Float(A=&RandomRangeFloat(Minimum=-360, Maximum=360), B=2))
   &Drag(
    Drag=&FloatFromCurve(
     FloatCurve=DataInterface() {propertyValues {Curve {
        defaultValue=1
        keys [{interpMode=RCIM_Cubic, tangentMode=RCTM_User, time=0, value=0}, {interpMode=RCIM_Cubic, tangentMode=RCTM_User, time=1, value=4}]
       }}}
     Scale Curve=0.5
    )
    Use Linear Drag=true
    Use Rotational Drag=true
   )
   &AccelerationForce(Acceleration=&Multiply_VectorByFloat(
     Vector=&RandomRangeVector(Minimum=<0 -250 50>, Maximum=<1 250 250>)
     Float=&FloatFromCurve(FloatCurve=DataInterface() {propertyValues {Curve {
         defaultValue=1
         keys [{interpMode=RCIM_Cubic, tangentMode=RCTM_User, time=0, value=0}, {interpMode=RCIM_Cubic, tangentMode=RCTM_User, time=0.2, value=0.5}]
        }}})
    ))
   &InheritSourceMovement(
    Applied Position Scale=&Multiply_VectorByFloat(
     Vector=<1 1 1>
     Float=&FloatFromCurve(FloatCurve=DataInterface() {propertyValues {Curve {
         defaultValue=1
         keys [{interpMode=RCIM_Cubic, tangentMode=RCTM_User, time=0, value=1}, {interpMode=RCIM_Cubic, tangentMode=RCTM_User, time=1, value=0}]
        }}})
    )
    ApplyForce=false
    ApplyPosition=true
    ApplyPositionFalloff=false
   ) {enabled=false}
   &CurlNoiseForce(Bias Noise=false, Noise Frequency=60, Noise Strength=100, Pan Noise=false)
   &SolveForcesAndVelocity(Clamp Velocity=false, Limit Acceleration=false)
  ]
  rendererClasses [&NiagaraSpriteRendererProperties]
  renderers [
   {rendererClass=&NiagaraSpriteRendererProperties}
   {rendererClass=&NiagaraSpriteRendererProperties, rendererIndex=1}
  ]
 }
 NE_SS_Trail {
  emitterSpawnScript [/SetVariables(Emitter.Linear Color:LinearColor=&MakeCustomLinearColorFromBool(
     Bool=&SetBoolByFloatComparison(A=&MakeFloatFromInt(Int=$User.ColorVariantID), B=14, Comparison Type=::NewEnumerator4)
     True Color=<rgba 1 1 1 1>
     False Color=<rgba 0 0 0 1>
    )) {isSetParameters=true}]
  emitterUpdateScript [
   &EmitterState()
   &SpawnRate(SpawnRate=180, Use Spawn Probability=false)
  ]
  particleSpawnScript [
   &VC_PaintedVariantsOverride(
    Black=<rgba 0.0296 0.0343 0.0382 1>
    BurntSienna=<rgba 0.25 0.042 0.0377 1>
    Cobalt=<rgba 0.0437 0.1384 1 1>
    Crimson=<rgba 0.7231 0.0242 0.0437 1>
    DefaultColor=&SelectLinearColorFromArray(
     Array Sampling Mode=::NewEnumerator0
     Direct Array Index=?
     Color Selection Array=DataInterface() {propertyValues {ColorData [<rgba 0.8 0.424 0.048 1>, <rgba 0.4951 0.0063 0.83 1>]}}
     Random Seed=0
    )
    ForestGreen=<rgba 0.003 1 0.0931 1>
    Grey=<rgba 0.3419 0.3467 0.3564 1>
    Lime=<rgba 0.5583 0.9823 0.1384 1>
    Orange=<rgba 0.6795 0.2051 0.0307 1>
    Pink=<rgba 0.9047 0.0497 0.7157 1>
    Purple=<rgba 0.261 0.0025 0.65 1>
    R1 Black:LinearColor=<rgba 0 0 0 1>
    R1 BurntSienna:LinearColor=<rgba 0 0 0 1>
    R1 Cobalt:LinearColor=<rgba 0 0 0 1>
    R1 Crimson:LinearColor=<rgba 0 0 0 1>
    R1 DefaultColor=<rgba 0 0 0 1>
    R1 ForestGreen:LinearColor=<rgba 0 0 0 1>
    R1 Gold:LinearColor=<rgba 0 0 0 1>
    R1 Grey:LinearColor=<rgba 0 0 0 1>
    R1 Lime:LinearColor=<rgba 0 0 0 1>
    R1 Orange:LinearColor=<rgba 0 0 0 1>
    R1 Pink:LinearColor=<rgba 0 0 0 1>
    R1 Purple:LinearColor=<rgba 0 0 0 1>
    R1 Saffron:LinearColor=<rgba 0 0 0 1>
    R1 SkyBlue:LinearColor=<rgba 0 0 0 1>
    Saffron=<rgba 0.8595 0.9 0.1136 1>
    SkyBlue=<rgba 0.0958 0.8513 0.9 1>
    TitaniumWhite=<rgba 0 0 0 1>
    Use R1 Painted Overrides=true
   )
   &InitializeParticle(
    Color Mode=::NewEnumerator0
    Sprite Rotation Mode=::NewEnumerator2
    Sprite Size Mode=::NewEnumerator1
    Uniform Sprite Size Max=70
    Uniform Sprite Size Min=50
    UsePositionOffset=false
    Write Material Random=true
   )
   &PartitionParticles()
   &SkeletalMeshLocation(
    Bone / Socket Index=&Add_Integer(A=$Particles.PartitionParticles.Partition, B=4)
    Bone / Socket Sampling Mode=::NewEnumerator5
    Interpolate Mesh Sampling=false
    MaskSpawnGroup=false
    Mesh Sampling Type=::NewEnumerator6
    Sampled Position Offset=&AddVectorToPosition(Position=$User.offset, Vector=<-64 0 -23>)
    Skeletal Mesh=$System.SkeletalMesh
   )
   &ShapeLocation(Hemisphere Distribution=<5 1>, MaskSpawnGroup=false, Override Seed=false, Sphere Radius=5, UseOwnerScale=true)
   &AddVelocity(Velocity=&VC_Mask_Vectors(
     FrontEnd Value:Vector3f=&RandomRangeVector(Minimum=<-300 -10 2>, Maximum=<-400 10 10>)
     blsFrontEndPreview:NiagaraBool=true
     SpeedMin Value:Vector3f=<-800 0 0>
     SpeedMax Value:Vector3f=<-1 0 0>
     Speed Alpha:NiagaraFloat=$System.Speed_Alpha
    ))
  ]
  particleUpdateScript [
   &FN_DisableTireFX(
    Mask Index Attribute:NiagaraInt32=$Particles.PartitionParticles.Partition
    Mask Index FL=2
    Mask Index FR=3
    Mask Index RL=0
    Mask Index RR=1
   )
   &ParticleState()
   &SpriteRotationRate(Rotation Rate=&Multiply_Float(A=&RandomRangeFloat(Minimum=-360, Maximum=360), B=2))
   &ScaleSpriteSize(Uniform Curve Sprite Scale=DataInterface() {propertyValues {Curve {
       defaultValue=0
       keys [
        {interpMode=RCIM_Cubic, tangentMode=RCTM_User, time=-0.0033, value=0.613}
        {arriveTangent=0.0607, interpMode=RCIM_Cubic, leaveTangent=-1.5756, tangentMode=RCTM_Break, time=0.4, value=1}
        <at 1 0>
       ]
      }}})
   &ScaleColor(
    Scale Alpha=&Lerp_Float(
     A=0
     B=4
     Alpha=/Niagara/DynamicInputs/Camera/CameraFade(
      Camera Query:NiagaraDataInterfaceCamera=DataInterface()
      Fade Offset:NiagaraFloat=25
      Fade Distance:NiagaraFloat=150
      Use Owner Position:NiagaraBool=true
     )
    )
    ScaleA=true
    ScaleRGB=true
   )
   &Drag(
    Drag=&FloatFromCurve(
     FloatCurve=DataInterface() {propertyValues {Curve {
        defaultValue=1
        keys [{interpMode=RCIM_Cubic, tangentMode=RCTM_User, time=0, value=0}, {interpMode=RCIM_Cubic, tangentMode=RCTM_User, time=1, value=4}]
       }}}
     Scale Curve=0.2
    )
    Use Linear Drag=true
    Use Rotational Drag=true
   )
   &AccelerationForce(Acceleration=&Multiply_VectorByFloat(
     Vector=&RandomRangeVector(Minimum=<-100 -250 50>, Maximum=<-100 250 250>)
     Float=&FloatFromCurve(FloatCurve=DataInterface() {propertyValues {Curve {
         defaultValue=1
         keys [{interpMode=RCIM_Cubic, tangentMode=RCTM_User, time=0, value=0}, {interpMode=RCIM_Cubic, tangentMode=RCTM_User, time=0.2, value=0.5}]
        }}})
    ))
   &InheritSourceMovement(
    Applied Position Scale=&Multiply_VectorByFloat(
     Vector=<1 1 1>
     Float=&FloatFromCurve(FloatCurve=DataInterface() {propertyValues {Curve {
         defaultValue=1
         keys [{interpMode=RCIM_Cubic, tangentMode=RCTM_User, time=0, value=1}, {interpMode=RCIM_Cubic, tangentMode=RCTM_User, time=1, value=0}]
        }}})
    )
    ApplyForce=false
    ApplyPosition=true
    ApplyPositionFalloff=false
   )
   &SolveForcesAndVelocity(Clamp Velocity=false, Limit Acceleration=false)
   &CameraOffset() {enabled=false}
  ]
  rendererClasses [&NiagaraSpriteRendererProperties]
  renderers [
   {rendererClass=&NiagaraSpriteRendererProperties}
   {rendererClass=&NiagaraSpriteRendererProperties, rendererIndex=1}
  ]
 }
 NE_SS_WheelWrap {
  emitterSpawnScript [/SetVariables(Emitter.ISPAINTED:NiagaraBool=&Boolean_NotOperation(Bool=$User.bIsPainted)) {isSetParameters=true}]
  emitterUpdateScript [
   &EmitterState(
    Inactive Response=::NewEnumerator1
    Life Cycle Mode=::NewEnumerator1
    Loop Behavior=::NewEnumerator1
    Loop Duration Mode=::NewEnumerator1
    UseLoopDelay=false
   )
   &SpawnBurst_Instantaneous(Spawn Count=2, Use Spawn Probability=false, UseLoopCountLimit=false)
  ]
  particleSpawnScript [
   &VC_PaintedVariantsOverride(
    Black=<rgba 0.0176 0.0176 0.0176 1>
    BurntSienna=<rgba 0.12 0.0269 0.0269 1>
    Crimson=<rgba 0.35 0.0107 0.0074 1>
    DefaultColor=<rgba 0.1611 0.002 0.27 1>
    Grey=<rgba 0.12 0.12 0.12 1>
    Lime=<rgba 0.2552 0.5926 0.0429 1>
    Orange=<rgba 0.3429 0.0953 0.0051 1>
    Pink=<rgba 0.5391 0.0217 0.4251 1>
    Purple=<rgba 0.1554 0 0.3786 1>
    R1 Black:LinearColor=<rgba 0.0176 0.0176 0.0176 1>
    R1 BurntSienna:LinearColor=<rgba 0.12 0.0269 0.0269 1>
    R1 Crimson:LinearColor=<rgba 0.35 0.0107 0.0074 1>
    R1 Grey:LinearColor=<rgba 0.12 0.12 0.12 1>
    R1 Lime:LinearColor=<rgba 0.2552 0.5926 0.0429 1>
    R1 Orange:LinearColor=<rgba 0.3429 0.0953 0.0051 1>
    R1 Pink:LinearColor=<rgba 0.5391 0.0217 0.4251 1>
    R1 Purple:LinearColor=<rgba 0.1554 0 0.3786 1>
    R1 Saffron:LinearColor=<rgba 0.797 0.8424 0.0314 1>
    R1 SkyBlue:LinearColor=<rgba 0.0577 0.3235 0.441 1>
    R1 TitaniumWhite:LinearColor=<rgba 0 0 0 5>
    R2 Black:LinearColor=<rgba 0 0 0 5>
    R2 BurntSienna:LinearColor=<rgba 0 0 0 1>
    R2 Cobalt:LinearColor=<rgba 0 0 0 1>
    R2 Crimson:LinearColor=<rgba 0 0 0 1>
    R2 DefaultColor:LinearColor=<rgba 0 0 0 1>
    R2 ForestGreen:LinearColor=<rgba 0 0 0 1>
    R2 Gold:LinearColor=<rgba 0 0 0 1>
    R2 Grey:LinearColor=<rgba 0 0 0 1>
    R2 Lime:LinearColor=<rgba 0 0 0 1>
    R2 Orange:LinearColor=<rgba 0 0 0 1>
    R2 Pink:LinearColor=<rgba 0 0 0 1>
    R2 Purple:LinearColor=<rgba 0 0 0 1>
    R2 Saffron:LinearColor=<rgba 0 0 0 1>
    R2 SkyBlue:LinearColor=<rgba 0 0 0 1>
    Saffron=<rgba 0.797 0.8424 0.0314 1>
    SkyBlue=<rgba 0.0151 0.0847 0.1155 1>
    TitaniumWhite=<rgba 0 0 0 1>
    Use R1 Painted Overrides=true
    Use R2 Painted Overrides=true
   )
   &InitializeParticle(
    Color Mode=::NewEnumerator0
    Lifetime=&MergeHelper_FloatSwitch(
     0=0.2
     1=&RandomRangeFloat(Minimum=0.5, Maximum=1)
     2=?
     3=?
     Index=&SpawnGroup()
     Num=::NewEnumerator0
    )
    Mesh Scale Mode=::NewEnumerator4
    Sprite Rotation Mode=::NewEnumerator2
    Sprite Size Mode=::NewEnumerator1
    Sprite UV Mode=::NewEnumerator3
    Uniform Sprite Size Max=25
    Uniform Sprite Size Min=20
    UsePositionOffset=false
    Write Material Random=true
   )
   &PartitionParticles()
   &SkeletalMeshLocation(
    Bone / Socket Index=&Add_Integer(A=$Particles.PartitionParticles.Partition, B=4)
    Bone / Socket Sampling Mode=::NewEnumerator5
    Interpolate Mesh Sampling=false
    MaskSpawnGroup=false
    Mesh Sampling Type=::NewEnumerator6
    Skeletal Mesh=$System.SkeletalMesh
   )
  ]
  particleUpdateScript [
   &FN_DisableTireFX(
    Mask Index Attribute:NiagaraInt32=$Particles.PartitionParticles.Partition
    Mask Index FL=2
    Mask Index FR=3
    Mask Index RL=0
    Mask Index RR=1
   )
   &VC_WheelWrapScale(Wheel Index Attribute:NiagaraInt32=&Add_Integer(A=$Particles.PartitionParticles.Partition, B=2))
   &ParticleState(Kill Particles When Lifetime Has Elapsed=false)
   &UpdateMeshOrientation(Rotation Coordinate Space=::NewEnumerator2, Rotation Vector=<0 1 0>) {enabled=false}
   &SkeletalMeshLocation(
    Bone / Socket Index=&Add_Integer(A=$Particles.PartitionParticles.Partition, B=4)
    Bone / Socket Sampling Mode=::NewEnumerator5
    MaskSpawnGroup=false
    Mesh Sampling Type=::NewEnumerator6
    Skeletal Mesh=$System.SkeletalMesh
   ) {module=SkeletalMeshLocation001}
   &SolveForcesAndVelocity(Clamp Velocity=false, Limit Acceleration=false) {enabled=false}
   &ScaleColor(Scale Alpha=0.4, Scale RGB=&VectorFromFloat(Value=0.7), ScaleA=true, ScaleRGB=true)
   /SetVariables(
    Particles.MaskOffset=-0.1
    Particles.Outlines2:LinearColor=&MultiplyLinearColor(
     Linear Color=&MakeCustomLinearColorFromBool(
      Bool=$User.bIsPainted
      True Color=$Particles.Initial.R1 Color
      False Color=&SelectLinearColorFromArray(
       Array Sampling Mode=::NewEnumerator2
       Direct Array Index=?
       Evaluation Type=::NewEnumerator1
       Interpolation=&FloatFromCurve(
        FloatCurve=DataInterface() {propertyValues {Curve {
           defaultValue=0
           keys [
            {interpMode=RCIM_Cubic, tangentMode=RCTM_User, time=0, value=0}
            {interpMode=RCIM_Cubic, tangentMode=RCTM_User, time=0.5, value=1}
            {tangentMode=RCTM_User, time=1, value=0}
           ]
          }}}
        CurveIndex=&Waveform(
         "Waveform [1]"=::NewEnumerator3
         "[1] Frequency"=&RandomRangeFloat(Minimum=-1, Maximum=1)
         "[1] Amplitude Min/Max"=<0 1>
        )
       )
       Color Selection Array=DataInterface() {propertyValues {ColorData [<rgba 0.2983 0.0038 0.5 1>, <rgba 0.5 0.1667 0 1>]}}
       Random Seed=?
      )
     )
     Scale Factor=&MakeVector4FromVectorAndFloat(
      "Vector (RGB)"=<1 1 1>
      "Float (Alpha)"=&SelectFloatFromArray(
       Array Sampling Mode=::NewEnumerator2
       Direct Array Index=?
       Interpolation=&Waveform("Waveform [1]"=::NewEnumerator3, "[1] Phase Scale"=0.2, "[1] Frequency"=1, "[1] Amplitude Min/Max"=<0 1>)
       Float Selection Array=DataInterface() {propertyValues {FloatData [0,1]}}
      )
     )
    )
    Particles.WHEELOUTLINES:LinearColor=&MultiplyLinearColor(
     Linear Color=&MakeCustomLinearColorFromBool(
      Bool=$User.bIsPainted
      True Color=$Particles.Initial.R1 Color
      False Color=&SelectLinearColorFromArray(
       Array Sampling Mode=::NewEnumerator2
       Direct Array Index=?
       Evaluation Type=::NewEnumerator1
       Interpolation=&FloatFromCurve(
        FloatCurve=DataInterface() {propertyValues {Curve {
           defaultValue=0
           keys [
            {interpMode=RCIM_Cubic, tangentMode=RCTM_User, time=0, value=0}
            {interpMode=RCIM_Cubic, tangentMode=RCTM_User, time=0.5, value=1}
            {tangentMode=RCTM_User, time=1, value=0}
           ]
          }}}
        CurveIndex=&Waveform(
         "Waveform [1]"=::NewEnumerator3
         "[1] Phase Scale"=0.2
         "[1] Frequency"=&RandomRangeFloat(Minimum=-0.5, Maximum=0.5)
         "[1] Compound Frequency"=0.1
         "[1] Amplitude Min/Max"=<0 1>
        )
       )
       Color Selection Array=DataInterface() {propertyValues {ColorData [<rgba 0.5 0.1667 0 1>, <rgba 0.2967 0 0.5 1>]}}
       Random Seed=?
      )
     )
     Scale Factor=&MakeVector4FromVectorAndFloat("Vector (RGB)"=<1 1 1>, "Float (Alpha)"=1)
    )
   ) {isSetParameters=true}
  ]
  rendererClasses [&NiagaraMeshRendererProperties]
  renderers [
   {rendererClass=&NiagaraMeshRendererProperties, rendererIndex=0}
   {rendererClass=&NiagaraMeshRendererProperties, rendererIndex=1}
   {rendererClass=&NiagaraMeshRendererProperties, rendererIndex=2}
   {rendererClass=&NiagaraMeshRendererProperties, rendererIndex=3}
   {rendererClass=&NiagaraMeshRendererProperties, rendererIndex=4}
  ]
 }
}
summary {
 systemName=NS_DriftTrail_StylizedSmoke_MP
 emitters [
  {emitterName=NE_SS_WheelWrap, rendererClasses [&NiagaraMeshRendererProperties]}
  {emitterName=NE_SS_Trail, rendererClasses [&NiagaraSpriteRendererProperties]}
  {emitterName=NE_SS_Cone, rendererClasses [&NiagaraSpriteRendererProperties]}
 ]
 userVariables [
  User.bIsFrontendPreview {type=NiagaraBool}
  User.bIsPainted {type=NiagaraBool}
  User.ColorVariantID {type=NiagaraInt32}
  User.DTCarOffset {type=Vector3f, description="", defaultValue {struct=Vector3f, value=<0 0 0>}}
  User.offset {type=NiagaraPosition, description="", defaultValue {}}
  User.PaintedColor {type=LinearColor, description="", defaultValue {struct=LinearColor, value=<rgba 1 1 1 1>}}
  User.PaintedScalar {type=NiagaraFloat, defaultValue=1, description=""}
  User.VehicleFlags {type=NiagaraInt32, defaultValue=806289408}
  User.Wheel_Radius_Array {
   type=NiagaraDataInterfaceArrayFloat
   defaultValue {
    struct=NiagaraExt_VariableValue_DataInterface
    value {
     dataInterface=&self:NiagaraDataInterfaceArrayFloat_1
     dataInterfaceClass=&NiagaraDataInterfaceArrayFloat
    }
   }
  }
  User.Wheel_Width_Array {
   type=NiagaraDataInterfaceArrayFloat
   defaultValue {
    struct=NiagaraExt_VariableValue_DataInterface
    value {
     dataInterface=&self:NiagaraDataInterfaceArrayFloat_0
     dataInterfaceClass=&NiagaraDataInterfaceArrayFloat
    }
   }
  }
  User.WheelTouchingArray {
   type=NiagaraDataInterfaceArrayFloat
   defaultValue {
    struct=NiagaraExt_VariableValue_DataInterface
    value {
     dataInterface=&self:NiagaraDataInterfaceArrayFloat_2
     dataInterfaceClass=&NiagaraDataInterfaceArrayFloat
    }
   }
  }
 ]
}
